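; Base Blueprint for Rift Anomaly actors, handling effect zones, activation hooks and teardown for child anomalies.
; Setup runs on BeginPlay and client data changes, following the data's effect rule; overlap mode sizes a sphere.
; The server equips or removes ability sets on overlapping pawns; ending strips effects, clears spline FX and destroys the actor.
; A server helper gathers nearby anomaly actors per player, feeding overridable ping and compass marking functions.
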(event EventBeginPlay
  ;[1 1 1]
  (CallFunction|SetupActions))

(event OnComponentBeginOverlap(Overlap)
  (params (OverlappedComponent :type Primitive%20Component%20Object%20Reference) (OtherActor :type Actor%20Object%20Reference) (OtherComp :type Primitive%20Component%20Object%20Reference) (OtherBodyIndex :type Integer) (bFromSweep :type Boolean) (SweepResult :type Hit%20Result%20Structure%20%28by%20ref%29))
  (CallFunction|OnRiftAnomalyEnterZone OtherActor)
  (Networking|SwitchHasAuthority
    (:Authority
      (switch Utilities|FlowControl|Switch|SwitchonERAEffectsRules (Class|FortRiftAnomaliesData|GetEffectApplicationRules (Variables|RiftAnomaliesActorBase|GetRiftAnomalyData))
        (:ApplyViaOverlap
          (bind _asfort_pawn (Utilities|Casting|CastToFortPawn OtherActor))
          (CallFunction|ApplyRAEffects _asfort_pawn))))))

(event OnComponentEndOverlap(Overlap)
  (params (OverlappedComponent :type Primitive%20Component%20Object%20Reference) (OtherActor :type Actor%20Object%20Reference) (OtherComp :type Primitive%20Component%20Object%20Reference) (OtherBodyIndex :type Integer))
  (CallFunction|OnRiftAnomalyLeaveZone OtherActor)
  (Networking|SwitchHasAuthority
    (:Authority
      (switch Utilities|FlowControl|Switch|SwitchonERAEffectsRules (Class|FortRiftAnomaliesData|GetEffectApplicationRules (Variables|RiftAnomaliesActorBase|GetRiftAnomalyData))
        (:ApplyViaOverlap
          (bind _asfort_pawn (Utilities|Casting|CastToFortPawn OtherActor))
          (CallFunction|RemoveRAEffects _asfort_pawn))))))

(event EventEndRiftAnomaly
  (bind _splinefxcomps (Variables|Default|GetSplineFXComps))
  (Utilities|FlowControl|Sequence
    (:then_0
      (Utilities|IsValid (Variables|RiftAnomaliesActorBase|GetRiftAnomalyData)
        (:Is%20Valid
          (switch Utilities|FlowControl|Switch|SwitchonERAEffectsRules (Class|FortRiftAnomaliesData|GetEffectApplicationRules (Variables|RiftAnomaliesActorBase|GetRiftAnomalyData))
            (:ApplyToAllPlayers
              _)
            (:ApplyViaOverlap
              _))
          (Networking|SwitchHasAuthority
            (:Authority
              (bind _outfortplayerpawns (Fortnite|GetFortPlayerPawns))
              (for _array_element _outfortplayerpawns
                (CallFunction|RemoveRAEffects _array_element)))))))
    (:then_1
      (if (Utilities|Array|IsNotEmpty _splinefxcomps)
        (for _array_element_2 _splinefxcomps
          (Components|DestroyComponent _array_element_2))
        (Utilities|Array|Clear _splinefxcomps)))
    (:then_2
      (Components|Activation|Deactivate (Variables|Default|GetNS_MatchModifier_Spline_Particles)))
    (:then_3
      (CallFunction|OnRiftAnomalyDeactivate)
      (Actor|DestroyActor))))

(event Custom|OnRifitAnomalyActivate)

(event Custom|OnRiftAnomalyDeactivate)

(event Custom|OnRiftAnomalyEnterZone
  (params (Actor :type Actor%20Object%20Reference)))

(event Custom|OnRiftAnomalyLeaveZone
  (params (Actor :type Actor%20Object%20Reference)))

(event EventRiftAnonalyDataChanged
  (Networking|SwitchHasAuthority
    (:Remote
      (CallFunction|SetupActions))))

(fn ApplyRAEffects
  (params (Pawn :type Fort%20Pawn%20Object%20Reference))
  (bind _returnvalue (Ability|EquipFortAbilitySet Pawn (Class|FortRiftAnomaliesData|GetAbilitySet (Variables|RiftAnomaliesActorBase|GetRiftAnomalyData)) self))
  (Utilities|Map|Add (Variables|Default|GetAbilitySetHandles) (Variables|GetPawn) _returnvalue))

(fn FindClosestAnomalyActorToPing
  (params (SpawnedAnomalyActors :type Array%20of%20Actor%20Object%20References))
  (bind _cachednearestanomalyactor (LocalVariables|GetCachedNearestAnomalyActor))
  (bind _cachedactorsforcompassdisplay (LocalVariables|GetCachedActorsForCompassDisplay))
  (bind _returnvalue (Utilities|Array|FindItem _cachedactorsforcompassdisplay _cachednearestanomalyactor))
  (bind _cachedplayercontroller (LocalVariables|GetCachedPlayerController))
  (bind _returnvalue_2 (Actor|GetComponentbyClass _cachedplayercontroller "/Script/FortniteGame.AthenaMarkerComponent"))
  (bind _cachednearestanomalyactor_2 (LocalVariables|GetCachedNearestAnomalyActor))
  (bind _cachedactorsforcompassdisplay_2 (LocalVariables|GetCachedActorsForCompassDisplay))
  (if (ScalableFloat|GetValueasBool (Variables|UI|GetHF_DisplayNavIcons))
    (Networking|SwitchHasAuthority
      (:Authority
        (Utilities|Array|Clear (LocalVariables|GetCachedActorsForCompassDisplay))
        (for _array_element (Fortnite|GetAllFortPlayerControllers true true)
          (Utilities|IsValid _array_element
            (:Is%20Valid
              (bind _output_get (LocalVariables|SetCachedPlayerController _array_element))
              (Utilities|IsValid (Game|Player|GetPlayerPawn _output_get)
                (:Is%20Valid
                  (LocalVariables|SetCachedPlayerLocation (Transformation|GetActorLocation (Game|Player|GetPlayerPawn _output_get)))
                  (for _array_element_2 (Variables|GetSpawnedAnomalyActors)
                    ;[1 1 1] Cache nearby actors for compass display
                    (if (< (Math|Vector|Distance2D(Vector) (Transformation|GetActorLocation _array_element_2) (LocalVariables|GetCachedPlayerLocation)) (ScalableFloat|GetValueatLevel (Variables|UI|GetSF_CompassRadius) 1.0))
                      (Utilities|Array|Add (LocalVariables|GetCachedActorsForCompassDisplay) _array_element_2)
                      (else
                        _))
                    ;[1 1 1] Cache closest actor to ping
                    (Utilities|IsValid _cachednearestanomalyactor_2
                      (:Is%20Valid
                        (if (< (Math|Vector|Distance2D(Vector) (Transformation|GetActorLocation _array_element_2) (LocalVariables|GetCachedPlayerLocation)) (Math|Vector|Distance2D(Vector) (LocalVariables|GetCachedPlayerLocation) (Transformation|GetActorLocation _cachednearestanomalyactor_2)))
                          _))
                      (:Is%20Not%20Valid
                        _))
                    (LocalVariables|SetCachedNearestAnomalyActor _array_element_2))
                  (Utilities|FlowControl|Sequence
                    (:then_0
                      ;[1 1 1] Pass the actors to be marked with Ping or on Compass - children Anomalies will determine marker details in these overrideable functions
                      (Utilities|IsValid _returnvalue_2
                        (:Is%20Valid
                          (UI|MarkAnomalyItemforPlayer _returnvalue_2 (LocalVariables|GetCachedNearestAnomalyActor))
                          ;; Make sure the closest item is at the top of the array to be included in the items added to the compass
                          (if (Utilities|Array|IsValidIndex _cachedactorsforcompassdisplay _returnvalue)
                            (Utilities|Array|SwapArrayElements _cachedactorsforcompassdisplay_2 0 _returnvalue)
                            (else
                              (Utilities|Array|Insert _cachedactorsforcompassdisplay_2 (LocalVariables|GetCachedNearestAnomalyActor))))
                          (UI|MarkAnomalyItemonCompass (LocalVariables|GetCachedPlayerController) (LocalVariables|GetCachedActorsForCompassDisplay)))))
                    (:then_1
                      ; Reset variables before moving to next player
                      (LocalVariables|SetCachedNearestAnomalyActor 0)
                      (Utilities|Array|Clear (LocalVariables|GetCachedActorsForCompassDisplay)))))))))))))

(fn GetPointAtAngleOnCircle
  (params (Angle :type Float%20%28double-precision%29) (Radius :type Float%20%28double-precision%29))
  (return (Math|Vector|MakeVector (* (Math|Trig|Sin(Radians) Angle) Radius) (* (Math|Trig|Cos(Radians) Angle) Radius))))

(fn MarkAnomalyItemForPlayer
  (params (PlayerToPing :type Athena%20Marker%20Component%20Object%20Reference) (AnomalyActorToBeMarked :type Actor%20Object%20Reference)))

(fn MarkAnomalyItemOnCompass
  (params (PlayerToPing :type Fort%20Player%20Controller%20Object%20Reference) (CachedActorsForCompassDisplay :type Array%20of%20Actor%20Object%20References)))

(fn RemoveRAEffects
  (params (Pawn :type Fort%20Pawn%20Object%20Reference))
  (bind _pawn (Variables|GetPawn))
  (bind _abilitysethandles (Variables|Default|GetAbilitySetHandles))
  (bind (_value _returnvalue) (Utilities|Map|Find _abilitysethandles _pawn))
  (if _returnvalue
    (Ability|UnequipFortAbilitySet _value)
    (Utilities|Map|Remove (Variables|Default|GetAbilitySetHandles) (Variables|GetPawn))))

(fn SetRARadius
  (Components|Sphere|SetSphereRadius (Variables|Default|GetOverlap) (ScalableFloat|GetValueatLevel (Variables|Default|GetEffectRadius)) true))

(fn SetupActions
  (Utilities|FlowControl|Sequence
    (:then_0
      (switch Utilities|FlowControl|Switch|SwitchonERAEffectsRules (Class|FortRiftAnomaliesData|GetEffectApplicationRules (Variables|RiftAnomaliesActorBase|GetRiftAnomalyData))
        (:DontApplyEffects
          _)
        (:ApplyToAllPlayers
          (Utilities|FlowControl|Sequence
            (:then_0
              _)
            (:then_1
              (Networking|SwitchHasAuthority))))
        (:ApplyViaOverlap
          (CallFunction|SetRARadius)
          (Variables|Default|SetUseSpline? true)))
      (Collision|SetCollisionEnabled (Variables|Default|GetOverlap) "NoCollision"))
    (:then_2
      (CallFunction|OnRifitAnomalyActivate))))

(fn ConstructionScript)
